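# Unlit alpha-composite material for the Landmass erosion brush that packs single-channel height data from a render target into four channels.
# A Custom HLSL node snaps U to groups of four texels and point-addresses their centers, writing the red values of four horizontal neighbors into RGBA.
# RGB is split out directly, while alpha is divided by a large constant, saturated and inverted, apparently to suit alpha-composite output.
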Material /Landmass/Landscape/BlueprintBrushes/Materials/Erosion/Erosion_PackHeightData
graph {
 properties {blendMode=BLEND_AlphaComposite, pixelDepthOffsetMode=PDOM_Legacy, shadingModel=MSM_Unlit, usedShadingModels=Unlit}
 nodes [
  MaterialExpressionTextureCoordinate_0 = TextureCoordinate()
  MaterialExpressionTextureObjectParameter_0 = TextureObjectParameter() {parameterName=RT}
  MaterialExpressionScalarParameter_0 = ScalarParameter() {defaultValue=64, parameterName=SizeX}
  MaterialExpressionScalarParameter_1 = ScalarParameter() {defaultValue=64, parameterName=SizeY}
  n1 = Custom!(
   UV=@MaterialExpressionTextureCoordinate_0
   Tex=@MaterialExpressionTextureObjectParameter_0
   SizeX=@MaterialExpressionScalarParameter_0
   SizeY=@MaterialExpressionScalarParameter_1
  ) {
   code="float2 X_Offset = float2(1.0 / SizeX, 0);\r\n\r\nfloat2 TexelIndex = floor(UV * float2(SizeX, SizeY));\r\n\r\nfloat2 OutUV = (TexelIndex + 0.5) / float2(SizeX, SizeY);\r\n\r\nOutUV.x = floor((UV.x * SizeX) / 4) * 4;\r\nOutUV.x += 0.5;\r\n\r\nOutUV.x /= SizeX;\r\n\r\nfloat4 OutSample = Tex.SampleLevel(TexSampler, OutUV, 0).r;\r\n\r\nOutSample.r = Tex.SampleLevel(TexSampler, OutUV + (X_Offset * 0), 0).r;\r\nOutSample.g = Tex.SampleLevel(TexSampler, OutUV + (X_Offset * 1), 0).r;\r\nOutSample.b = Tex.SampleLevel(TexSampler, OutUV + (X_Offset * 2), 0).r;\r\nOutSample.a = Tex.SampleLevel(TexSampler, OutUV + (X_Offset * 3), 0).r;\r\n\r\nreturn OutSample;"
   outputType=CMOT_Float4
   showCode=true
   inputs [
    {
     inputName=UV
     input {expression=&self:MaterialExpressionTextureCoordinate_0}
    }
    {
     inputName=Tex
     input {expression=&self:MaterialExpressionTextureObjectParameter_0}
    }
    {
     inputName=SizeX
     input {expression=&self:MaterialExpressionScalarParameter_0}
    }
    {
     inputName=SizeY
     input {expression=&self:MaterialExpressionScalarParameter_1}
    }
   ]
  }
  ComponentMask(None=@n1) {b=true, g=true, r=true}
  n2 = ComponentMask(None=@n1) {a=true}
  OneMinus(None=Saturate(None=Divide(A=@n2) {constB=10000}))
 ]
}
